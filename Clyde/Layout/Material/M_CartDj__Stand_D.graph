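# Default Lit material for the leg/stand of a DJ cart, using the LemonCartDj_Leg diffuse, normal, mask and emissive textures.
# A procedural glow ring comes from subtracting two UV sphere masks at a shared center, then sharpening and boosting the result.
# The "Use GlowLine" scalar blends that ring in, and it is summed with the emissive texture and the Emissive output of MF_FloatWPO.
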Material /Clyde/Layout/Material/M_CartDj__Stand_D
graph {
 properties {bAutomaticallySetUsageInEditor=false, usedShadingModels=Default Lit}
 nodes [
  n1 = SphereMask(A=TextureCoordinate(), B=AppendVector(A=Constant() {r=0.292}, B=Constant() {r=0.54}), Radius=Constant() {r=0.18}) {attenuationRadius=0.2}
  n2 = SphereMask(A=TextureCoordinate(), B=AppendVector(A=Constant() {r=0.292}, B=Constant() {r=0.54}), Radius=Constant() {r=0.154}) {attenuationRadius=0.2, hardnessPercent=90}
  n3 = Multiply(A=Power(Base=Subtract(A=@n1, B=@n2)) {constExponent=30}) {constB=100}
  n4 = ScalarParameter() {parameterName=Use GlowLine}
  n5 = MaterialFunctionCall() {materialFunction=&/MF_FloatWPO}
  n6 = TextureSample() {texture=/Clyde/Layout/Texture/T_LemonCartDj_Leg_E}
  Reroute(None=Add(A=LinearInterpolate(B=@n3, Alpha=@n4), B=Add(A=@n5.Emissive, B=@n6.RGB)))
  TextureSample() {texture=/Clyde/Layout/Texture/T_LemonCartDj_Leg_D}
  TextureSample() {samplerType=SAMPLERTYPE_Normal, texture=/Clyde/Layout/Texture/T_LemonCartDj_Leg_N}
  TextureSample() {samplerType=SAMPLERTYPE_Masks, texture=/Clyde/Layout/Texture/T_LemonCartDj_Leg_S}
 ]
}
