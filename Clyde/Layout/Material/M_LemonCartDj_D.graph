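# M_LemonCartDj_D is a Default Lit surface material, apparently for the table of the lemon cart DJ layout prop.
# It samples the table's _D, normal and mask textures, whose suffixes suggest a standard base color, normal and surface-mask set.
# An Add node combines the RGB of the _E texture with the Emissive output of the MF_FloatWPO material function, layering that function's glow over the authored emissive map.
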Material /Clyde/Layout/Material/M_LemonCartDj_D
graph {
 properties {bAutomaticallySetUsageInEditor=false, usedShadingModels=Default Lit}
 nodes [
  Add(
   A=TextureSample() {texture=/Clyde/Layout/Texture/T_LemonCartDj_Table_E}.RGB
   B=MaterialFunctionCall() {materialFunction=&/MF_FloatWPO}.Emissive
  )
  TextureSample() {texture=/Clyde/Layout/Texture/T_LemonCartDj_Table_D}
  TextureSample() {samplerType=SAMPLERTYPE_Normal, texture=/Clyde/Layout/Texture/T_LemonCartDj_Table_N}
  TextureSample() {samplerType=SAMPLERTYPE_Masks, texture=/Clyde/Layout/Texture/T_LemonCartDj_Table_S}
 ]
}
